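# Niagara trail effect for the Keytar Boomerang emote, pairing a ribbon streak with spinning wind-whirl meshes, both spawned per unit of movement.
# User.NormalizedNotifyProgress shortens ribbon particle lifetime along a curve, while ribbons narrow and fade with age.
# The Windwhirl emitter uses User.WindWhirlSpawnRate as spawn probability, emitting randomly sized, rotating meshes that grow in and fade out.
# Both emitters scale alpha by time of day, dimmest at night.
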NiagaraSystem /SparksCosmetics/Emotes/Keytar/Boomerang/Effects/NS_Emote_Keytar_Boomerang_Trail
dependencies {usedDataInterfaces [&NiagaraDataInterfaceCurve], usedDynamicInputs [], usedModules [&SystemState]}
emitters {
 Trail {
  emitterUpdateScript [
   &EmitterState()
   &SpawnPerUnit(Module.Use Max Movement Threshold=true, Module.Use Movement Tolerance=false, Use Spawn Probability=false)
  ]
  particleSpawnScript [
   &InitializeParticle(
    Color Mode=::NewEnumerator0
    Lifetime=&FloatFromCurve(
     FloatCurve=DataInterface() {propertyValues {Curve {keys [
         <at 0 0.5 RCIM_Cubic>
         {arriveTangent=-0.45, interpMode=RCIM_Cubic, leaveTangent=-0.45, time=0.5, value=0.2}
         {arriveTangent=-0.9291, interpMode=RCIM_Cubic, leaveTangent=-0.9291, tangentMode=RCTM_User, time=1, value=0.05}
        ]}}}
     CurveIndex=$User.NormalizedNotifyProgress
    )
    Ribbon Facing Mode=::NewEnumerator1
    Ribbon Width=80
    Ribbon Width Mode=::NewEnumerator1
    Sprite Size Mode=::NewEnumerator3
    Uniform Sprite Size=1
    UsePositionOffset=false
    Write Material Random=true
   )
   &ApplyInitialForces()
  ]
  particleUpdateScript [
   &ParticleState()
   &SolveForcesAndVelocity(Clamp Velocity=false, Limit Acceleration=false)
   &ScaleRibbonWidth(Ribbon Width Scale=&FloatFromCurve(FloatCurve=DataInterface() {propertyValues {Curve {keys [<at 0 1>,<at 1 0.1>]}}}))
   &ScaleColor(
    Scale Alpha=&Multiply_Float(
     A=&FloatFromCurve(FloatCurve=DataInterface() {propertyValues {Curve {keys [
          {interpMode=RCIM_Cubic, tangentMode=RCTM_User, time=0, value=0}
          {interpMode=RCIM_Cubic, tangentMode=RCTM_User, time=0.03, value=1}
          {tangentMode=RCTM_User, time=1, value=0}
         ]}}})
     B=&FNTimeOfDayScalar(Night Scale=0.25, Morning Scale=0.275, Day Scale=0.3, Evening Scale=0.275)
    )
    ScaleA=true
    ScaleRGB=true
   )
  ]
  rendererClasses [&NiagaraRibbonRendererProperties]
  renderers [{rendererClass=&NiagaraRibbonRendererProperties}]
 }
 Windwhirl {
  emitterUpdateScript [
   &EmitterState(Inactive Response=::NewEnumerator1, Life Cycle Mode=::NewEnumerator1, Loop Behavior=::NewEnumerator1, Loop Duration=2.77, UseLoopDelay=false)
   &SpawnPerUnit(
    Module.Use Max Movement Threshold=true
    Module.Use Movement Tolerance=false
    Spawn Probability:NiagaraFloat=$User.WindWhirlSpawnRate
    Spawn Spacing=10
    Use Spawn Probability=true
   )
  ]
  particleSpawnScript [
   &InitializeParticle(
    Color=<rgba 0.6604 0.7749 0.8073 1>
    Lifetime Max=0.6
    Lifetime Min=0.3
    Lifetime Mode=::NewEnumerator1
    Mesh Scale Mode=::NewEnumerator1
    Mesh Uniform Scale Max=0.85
    Mesh Uniform Scale Min=0.4
    UsePositionOffset=false
    Write Material Random=true
   )
   &SystemLocation(MaskSpawnGroup=false, Offset=<0 0 15>)
   &ShapeLocation(
    Cylinder Height=10
    Cylinder Radius=5
    MaskSpawnGroup=false
    Override Seed=false
    Rotation Angle=90
    Rotation Axis=<0 1 0>
    Rotation Mode=::NewEnumerator1
    Shape Primitive=::NewEnumerator1
    Surface Only=false
    UseOwnerScale=true
   )
   &InitialMeshOrientation(
    Mesh Axis To Orient=<0 1 0>
    Mesh Orientation Mode=::NewEnumerator2
    Orientation Vector=<0 0 1>
    Rotation=&RandomRangeVector(Minimum=<0 0 0>, Maximum=<0 0 1>)
    Use Rotation Vector=true
   )
  ]
  particleUpdateScript [
   &ParticleState()
   &ScaleColor(
    Scale Alpha=&Multiply_Float(
     A=&FloatFromCurve(FloatCurve=DataInterface() {propertyValues {Curve {keys [
          {arriveTangent=4.7356, interpMode=RCIM_Cubic, leaveTangent=4.7356, tangentMode=RCTM_User, time=0, value=0}
          {interpMode=RCIM_Cubic, tangentMode=RCTM_User, time=0.5, value=1}
          <at 1 0 RCIM_Cubic>
         ]}}})
     B=&FNTimeOfDayScalar(Night Scale=0.1, Morning Scale=0.1125, Day Scale=0.125, Evening Scale=0.1125)
    )
    ScaleA=true
    ScaleRGB=true
   )
   &UpdateMeshOrientation(
    Rotation Coordinate Space=::NewEnumerator3
    Rotation Rate=&RandomRangeFloat(Minimum=5, Maximum=10)
    Rotation Vector=&RandomRangeVector(Minimum=<0 0 0>, Maximum=<0 0 1>)
   )
   &ScaleMeshSize(Scale Factor=&VectorFromFloat(Value=&FloatFromCurve(FloatCurve=DataInterface() {propertyValues {Curve {keys [{arriveTangent=3.0713, interpMode=RCIM_Cubic, leaveTangent=3.0713, tangentMode=RCTM_User, time=0, value=0}, <at 0.6 1 RCIM_Cubic>]}}})))
  ]
  rendererClasses [&NiagaraMeshRendererProperties]
  renderers [{rendererClass=&NiagaraMeshRendererProperties, rendererIndex=0}]
 }
}
summary {
 systemName=NS_Emote_Keytar_Boomerang_Trail
 emitters [
  {emitterName=Trail, rendererClasses [&NiagaraRibbonRendererProperties]}
  {emitterName=Windwhirl, rendererClasses [&NiagaraMeshRendererProperties]}
 ]
 userVariables [
  User.NormalizedNotifyProgress {type=NiagaraFloat, defaultValue=0, description=""}
  User.WindWhirlSpawnRate {type=NiagaraFloat, defaultValue=1, description=""}
 ]
}
